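# Default Lit water surface material for the Sparks lobby environment, usable on instanced and spline meshes.
# The shared two-texture water panner supplies animated normals, packed into material attributes with flat color, specular, roughness and opacity parameters.
# Its Normal output also feeds World Position Offset, and flatten functions branch off the normal.
# A Single Layer Water output with derived scattering and absorption remains, although the shading model is Default Lit.
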Material /SparksLobby/Environment/M_SparksOpaqueWater
graph {
 properties {
  bCastRayTracedShadows=false
  bUsedWithInstancedStaticMeshes=true
  bUsedWithSplineMeshes=true
  bUsedWithWater=true
  pixelDepthOffsetMode=PDOM_Legacy
  usedShadingModels=Default Lit
 }
 nodes [
  n1 = MaterialFunctionCall!(Normal, WPO, Far Mask Override, Additional Mask, LowDetailBiome_Mask) {materialFunction=/FortWater/Materials/Functions/WaterTwoTexturePanner}
  n2 = SetMaterialAttributes!(
   MaterialAttributes
   Base Color=VectorParameter() {parameterName=BaseColor, defaultValue=<rgba 0.0215 0.0376 0.0885 1>}.RGB
   Specular=ScalarParameter() {defaultValue=0.25, group=Water Shading Model, parameterName=Specular}
   Roughness=ScalarParameter() {defaultValue=0.4, group=Water Shading Model, parameterName=Roughness}
   Opacity=ScalarParameter() {group=Water Shading Model, parameterName=Surface Opacity}
   Normal=@n1.Normal
   World Position Offset=@n1.Normal
  ) {attributeSetTypes [
    #1
    #2
    #3
    #4
    #5
    #6
   ]}
  n3 = GetMaterialAttributes(None=@n2) {attributeGetTypes [
    #1
    #2
    #3
    #4
    #5
    #6
   ]}
  MaterialFunctionCall(Normal=Multiply(A=@n3.Normal, B=TwoSidedSign()), Flatness=Constant() {r=4}) {materialFunction=&FlattenNormal}
  MaterialFunctionCall!(Normal=@n3.Normal) {materialFunction=/FortWater/Materials/Functions/MF_WaveFlatten}
  n4 = VectorParameter() {group=Water Shading Model, parameterName=Scattering, defaultValue=<rgba 0.1268 0.1336 0.15 0.5>}
  n5 = VectorParameter() {group=Water Shading Model, parameterName=Absorption, defaultValue=<rgba 1 13.35 20 100>}
  SingleLayerWaterMaterialOutput!(
   ScatteringCoefficients=Multiply(A=@n4.RGB, B=Divide(A=@n4.A) {constB=100})
   AbsorptionCoefficients=Divide(B=Multiply(A=@n5.RGB, B=@n5.RGB))
   PhaseG=ScalarParameter() {defaultValue=0.3, group=Water Shading Model, parameterName=Anisotropy}
   ColorScaleBehindWater
  )
 ]
}
