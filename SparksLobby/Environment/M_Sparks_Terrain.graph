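# Landscape material for the Sparks lobby terrain, layering shared Helios grass and cliff functions driven by terrain masks, biome weights and a slope mask.
# The blend writes to runtime virtual textures and samples back by platform and quality, with Nanite displacement at Epic and derived-normal fallback, before post-RVT detail passes.
# It also outputs landscape grass and a grass physical material.
# A temporary IsLobbyIsland switch can override one terrain mask.
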Material /SparksLobby/Environment/M_Sparks_Terrain
graph {
 properties {
  bEnableDisplacementFade=true
  bEnableTessellation=true
  bTangentSpaceNormal=false
  bUseLightmapDirectionality=false
  bUseMaterialAttributes=true
  bUsedWithNanite=true
  bUsedWithVirtualHeightfieldMesh=true
  pixelDepthOffsetMode=PDOM_Legacy
  shadingModel=MSM_FromMaterialExpression
  usedShadingModels=Default Lit
  displacementFadeRange {endSizePixels=3, startSizePixels=8}
  displacementScaling {center=0.5, magnitude=0.15}
 }
 nodes [
  n1 = MaterialFunctionCall!(
   Material
   MobileMaterial
   IsForStaticMesh
   UseTangentNormals
   UseCustomSlopeMask
   CustomSlopeMask
   Grass_EdgeOnly=NamedRerouteUsage()
   LayerWeight=NamedRerouteUsage()
   TerrainMasks_QuadA=NamedRerouteUsage()
   TerrainMasks_QuadB=NamedRerouteUsage()
   TerrainMasks_QuadC=NamedRerouteUsage()
   TerrainMasks_QuadD=NamedRerouteUsage()
   VariationMasks Mobile=NamedRerouteUsage()
  ) {materialFunction=/Game/Athena/Helios/Environments/Landscape/Materials/LandscapeLayers/MF_Helios_Grass}
  n2 = MaterialFunctionCall!(
   Material=@n1.BlendedMaterial
   MobileMaterial=@n1.BlendedMaterial
   LayerWeight=NamedRerouteUsage()
   UseTangentNormals
   TerrainMasks_QuadA=NamedRerouteUsage()
   TerrainMasks_QuadB=NamedRerouteUsage()
   TerrainMasks_QuadC=NamedRerouteUsage()
   TerrainMasks_QuadD=NamedRerouteUsage()
   VariationMasks Mobile=NamedRerouteUsage()
   MossDesaturateMask
  ) {materialFunction=/Game/Athena/Helios/Environments/Landscape/Materials/LandscapeLayers/MF_Helios_Cliff}
  Composite!("HighEndMaterial(MA)"=@n2.BlendedMaterial, "MobileMaterial(MA)"=@n2.BlendedMaterial)
  Composite!() {subgraphName=Grass_Physmat_Mask}
  n3 = NamedRerouteUsage()
  MaterialExpressionMaterialFunctionCall_43 = MaterialFunctionCall!(In=Multiply(A=NamedRerouteUsage(), B=OneMinus(None=@n3)), ZSpawnLimit) {materialFunction=/Game/Athena/Helios/Environments/Landscape/Materials/MaterialFunctions_Utilities/MF_Helios_LS_GrassSwitch}
  LandscapeGrassOutput!(Shared_Grass_Layer01=@MaterialExpressionMaterialFunctionCall_43.Result) {grassTypes [Shared_Grass_Layer01 {
     grassType=&/LS_GrassLayer_Sparks
     input {
      expression=&self:MaterialExpressionMaterialFunctionCall_43
      inputName=None
      mask=0
      maskA=0
      maskB=0
      maskG=0
      maskR=0
      outputIndex=0
     }
    }]}
  n4 = NamedRerouteUsage()
  n5 = NamedRerouteUsage()
  n6 = Max(A=Max(A=NamedRerouteUsage(), B=NamedRerouteUsage()), B=NamedRerouteUsage())
  n7 = Constant() {desc=Replace With SnowLayer_Alpha}
  n8 = Multiply(A=Max(A=NamedRerouteUsage(), B=Max(A=@n4, B=@n5)), B=OneMinus(None=Max(A=@n6, B=@n7)))
  MaterialExpressionReroute_11 = Reroute(None=@n8)
  LandscapePhysicalMaterialOutput!(Grass=@MaterialExpressionReroute_11.Output) {
   desc="These are layered in order of importance.\r\nIf Dirt=1 and Stone=1, then Dirt will \"win\".\r\nTheoretically, these values aren't clamped.\r\nSo a value of 2 should trump a value of 1."
   inputs [{
     physicalMaterial=/Game/FortressPhysicalMaterials/Grass
     input {
      expression=&self:MaterialExpressionReroute_11
      inputName=None
      mask=0
      maskA=0
      maskB=0
      maskG=0
      maskR=0
      outputIndex=0
     }
    }]
  }
  n9 = RuntimeVirtualTextureSampleParameter!(UVs, Translated World Position) {
   group=RVT
   materialType=BaseColor_Normal_Specular_YCoCg
   parameterName=RVT Sample Target
   virtualTexture=/Game/Athena/Apollo/Maps/Landscape/VT/RVT_Apollo_Landscape
   worldPositionOriginType=CameraRelative
  }
  n10 = SetMaterialAttributes!(MaterialAttributes, Base Color=@n9.BaseColor, Specular=@n9.BaseColor, Roughness=@n9.BaseColor, Normal=@n9.BaseColor) {attributeSetTypes [
    #1
    #2
    #3
    #4
   ]}
  n11 = RuntimeVirtualTextureSampleParameter() {
   group=RVT
   materialType=Displacement
   parameterName=RVT Displacement
   virtualTexture=/Game/Athena/Apollo/Maps/Landscape/VT/RVT_Helios_Displacement
   worldPositionOriginType=CameraRelative
  }
  n12 = SetMaterialAttributes!(
   MaterialAttributes
   Base Color=@n9.BaseColor
   Specular=@n9.BaseColor
   Roughness=@n9.BaseColor
   Normal=@n9.BaseColor
   Displacement=ShadingPathSwitch(Default=@n11.Displacement)
  ) {attributeSetTypes [
    #1
    #2
    #3
    #4
    #5
   ]}
  MaterialExpressionReroute_21 = Reroute(None=@n2.MobileBlendedMaterial)
  n13 = MaterialFunctionCall!(MobileMaterial=@MaterialExpressionReroute_21."MobileMaterial(MA)", TransformTangents) {materialFunction=/Game/Athena/Helios/Environments/Landscape/Materials/MaterialFunctions_Utilities/MF_MobileDeriveNormals}
  n14 = VirtualTextureFeatureSwitch!(No=@n13.ResultMaterial, Yes=QualitySwitch(Default=@n10, Epic=@n12))
  n15 = ShadingPathSwitch(Default=@n14, Mobile=QualitySwitch(Default=@n13.ResultMaterial, Medium=@n14))
  MaterialExpressionReroute_18 = Reroute(None=@n15)
  n16 = MaterialFunctionCall!(
   Material=@MaterialExpressionReroute_18.Output
   UseSnowForBiomeC
   Snow_PostRVT
   SwitchGrassLayerMask=NamedRerouteUsage()
   TerrainMasks_QuadA
   TerrainMasks_QuadB
   TerrainMasks_QuadC
   TerrainMasks_QuadD
  ) {
   desc=For Post RVT Details on landscape layers
   materialFunction=/Game/Athena/Helios/Environments/Landscape/Materials/MaterialFunctions_Utilities/PostRVT_LayerDetails
  }
  MaterialFunctionCall!(LandscapeMaterial=@n16.Result, VariationMasksMobile=NamedRerouteUsage()) {
   desc=For Common Gameplay/Misc Post RVT Details
   materialFunction=/Game/Athena/Helios/Environments/Landscape/Materials/MaterialFunctions_Utilities/MF_PostRVT_Misc
  }
  MaterialExpressionReroute_20 = Reroute(None=@n2.BlendedMaterial)
  n17 = GetMaterialAttributes(None=ShadingPathSwitch(Default=@MaterialExpressionReroute_20."HighEndMaterial(MA)", Mobile=@n13.ResultMaterial)) {attributeGetTypes [
    #1
    #2
    #3
    #4
    #5
   ]}
  Multiply(A=@n17.Displacement, B=ScalarParameter() {defaultValue=1, group=Displacement, parameterName=Displacement_BASE_Mod})
  n18 = StaticBoolParameter() {
   desc="Temp workaround for LobbyIsland so at least the layers can\r\nbe painted until the RVT is setup correctly"
   group=Temp
   parameterName=IsLobbyIsland
  }
  NamedRerouteDeclaration(None=@n18) {name=IsLobbyIsland, nodeColor=<rgba 0.5 0 1 1>}
  NamedRerouteDeclaration(None=@n1.LayerMask) {name=GrassLayerMask, nodeColor=<rgba 0.5 0 1 1>}
  NamedRerouteDeclaration(None=@n2.GrassMask) {name=CliffLayerGrassMask, nodeColor=<rgba 0.6172 0 1 1>}
  n19 = Constant() {r=1}
  n20 = Constant()
  n21 = MaterialFunctionCall!(UseLandscapeCoords, "UseProjection(translucent)") {materialFunction=/Game/Athena/Helios/Environments/Landscape/Materials/MaterialFunctions_Utilities/MF_TerrainMasks}
  NamedRerouteDeclaration(None=StaticSwitch(
    True=MaterialFunctionCall(X=@n19, Y=@n20, Z=@n20, A=@n20) {materialFunction=&MakeFloat4}.Result
    False=@n21.QuadB_RGBA
    Value=NamedRerouteUsage()
   )) {name=TerrainMasks_QuadB, nodeColor=<rgba 0.5 0 1 1>}
  NamedRerouteDeclaration(None=@n21.QuadA_RGBA) {name=TerrainMasks_QuadA, nodeColor=<rgba 0.5 0 1 1>}
  NamedRerouteDeclaration(None=@n21.QuadC_RGBA) {name=TerrainMasks_QuadC, nodeColor=<rgba 0.5 0 1 1>}
  NamedRerouteDeclaration(None=@n21.QuadD_RGBA) {name=TerrainMasks_QuadD, nodeColor=<rgba 0.5 0 1 1>}
  n22 = LandscapeLayerSample() {parameterName=BiomeA}
  n23 = LandscapeLayerSample() {parameterName=BiomeB}
  n24 = LandscapeLayerSample() {parameterName=BiomeC}
  n25 = LandscapeLayerSample() {parameterName=BiomeD}
  n26 = Max(
   A=Max(
    A=NamedRerouteDeclaration(None=@n22) {name=BiomeA_Weight, nodeColor=<rgba 0.5 0 1 1>}
    B=NamedRerouteDeclaration(None=@n23) {name=BiomeB_Weight, nodeColor=<rgba 0.5 0 1 1>}
   )
   B=Max(
    A=NamedRerouteDeclaration(None=@n24) {name=BiomeC_Weight, nodeColor=<rgba 0.5 0 1 1>}
    B=NamedRerouteDeclaration(None=@n25) {name=BiomeD_Weight, nodeColor=<rgba 0.5 0 1 1>}
   )
  )
  NamedRerouteDeclaration(None=OneMinus(None=@n26)) {name=Grass_Weight_NoCliff, nodeColor=<rgba 0.5 0 1 1>}
  NamedRerouteDeclaration(None=MaterialFunctionCall() {materialFunction=/Game/Athena/Helios/Environments/Landscape/Materials/MaterialFunctions_Utilities/MF_VariationMasks_Mobile}.VariationMasks Mobile) {name=VariationMasksMobile, nodeColor=<rgba 1 0.9141 0 1>}
  NamedRerouteDeclaration(None=@n26) {name=Grass_EdgeOnly, nodeColor=<rgba 0.5 0 1 1>}
  NamedRerouteDeclaration(None=@n1.BlendedMaterial) {name=Layer0_DefaultGrass, nodeColor=<rgba 0.5 0 1 1>}
  NamedRerouteDeclaration(None=@n1.MobileBlendedMaterial) {name=Layer0_DefaultGrassMobile, nodeColor=<rgba 0.5 0 1 1>}
  n27 = NamedRerouteDeclaration(None=MaterialFunctionCall() {materialFunction=/Game/Athena/Helios/Environments/Landscape/Materials/MaterialFunctions_Utilities/MF_CliffSlopeMask}.Result) {name=Cliff_Weight, nodeColor=<rgba 0.5 0 1 1>}
  NamedRerouteDeclaration(None=OneMinus(None=Max(A=@n26, B=@n27))) {name=Grass_Weight, nodeColor=<rgba 0.5 0 1 1>}
  PinBase()
  PinBase() {reroutePins [Output {expression=null}]}
  PinBase()
  PinBase() {reroutePins [Output {expression=null}]}
  PinBase()
  PinBase!(Output=@n8) {reroutePins [Output {expression=&self:MaterialExpressionReroute_11}]}
  PinBase!()
  PinBase() {reroutePins [Output {expression=null}]}
  PinBase()
  PinBase() {reroutePins [Output {expression=null}]}
  PinBase() {reroutePins [
    "HighEndMaterial(MA)" {expression=&self:MaterialExpressionReroute_20}
    "MobileMaterial(MA)" {expression=&self:MaterialExpressionReroute_21}
   ]}
  PinBase!(Output=@n15) {reroutePins [Output {expression=&self:MaterialExpressionReroute_18}]}
  n28 = ScalarParameter() {
   desc="A value of zero = default displacement input. A value of 1 = a scalar displacement value of .5 which means no displacement."
   group=Displacement
   parameterName=Displacement_BASE_DEBUG
   sortPriority=-9999
  }
  RuntimeVirtualTextureOutput!(
   BaseColor=@n17.Base Color
   Specular=@n17.Base Color
   Roughness=@n17.Base Color
   Normal=@n17.Base Color
   WorldHeight=WorldPosition().Z
   Opacity
   Mask
   Displacement=LinearInterpolate(A=@n17.Displacement, Alpha=Saturate(None=@n28)) {constB=0.5}
   Mask4
  )
 ]
}
